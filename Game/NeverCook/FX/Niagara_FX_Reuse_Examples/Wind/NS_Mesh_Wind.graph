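# Niagara mesh wind effect from the FX reuse examples, built as a single Mesh_Wind emitter.
# It burst-spawns translucent mesh particles whose lifetime matches the emitter's loop time, with alpha scaled down during update.
# Dynamic material parameters multiply System.Age by per-channel rates, likely panning the material.
# Driving them from system age rather than particle age keeps that animation independent of each particle respawn.
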NiagaraSystem /Game/NeverCook/FX/Niagara_FX_Reuse_Examples/Wind/NS_Mesh_Wind
dependencies {usedDataInterfaces [&NiagaraDataInterfaceColorCurve], usedDynamicInputs [], usedModules [&SystemState]}
emitters {Mesh_Wind {
  emitterSpawnScript [/SetVariables(Emitter.LoopTime=2) {isSetParameters=true}]
  emitterUpdateScript [
   &EmitterState(Life Cycle Mode=::NewEnumerator1, Loop Duration=$Emitter.LoopTime, UseLoopDelay=false)
   &SpawnBurst_Instantaneous(Use Spawn Probability=false, UseLoopCountLimit=false)
  ]
  particleSpawnScript [&InitializeParticle(Color=<rgba 1 1 1 0.5>, Lifetime=$Emitter.LoopTime, Mesh Scale Mode=::NewEnumerator3, UsePositionOffset=false, Write Material Random=true)]
  particleUpdateScript [
   &ParticleState()
   &ScaleColor(Scale Alpha=0.25, ScaleA=true, ScaleRGB=true)
   &DynamicMaterialParameters(
    Index 0 Param 1=0.6
    Index 0 Param 4=1
    Index 1 Param 1=&Multiply_Float(A=1, B=$System.Age)
    Index 1 Param 2=&Multiply_Float(A=0.25, B=$System.Age)
    Index 1 Param 3=&Multiply_Float(A=0, B=$System.Age)
    Index 1 Param 4=&Multiply_Float(A=1, B=$System.Age)
    Index 2 Param 1=&Multiply_Float(A=0, B=$System.Age)
    Index 2 Param 2=&Multiply_Float(A=-1, B=$System.Age)
    Index 2 Param 3=&Multiply_Float(A=-1, B=$System.Age)
    Index 2 Param 4=&Multiply_Float(A=-0.2, B=$System.Age)
    Module.Param0WriteEnabled=true
    Module.Param0WriteEnabled1=true
    Module.Param0WriteEnabled2=true
    Module.Param1WriteEnabled=true
    Module.Param1WriteEnabled1=true
    Module.Param1WriteEnabled2=true
    Module.Param2WriteEnabled=true
    Module.Param2WriteEnabled1=true
    Module.Param2WriteEnabled2=true
    Module.Param3WriteEnabled=true
    Module.Param3WriteEnabled1=true
    Module.Param3WriteEnabled2=true
    Write Parameter Index 1=true
    Write Parameter Index 2=true
   )
  ]
  rendererClasses [&NiagaraMeshRendererProperties]
  renderers [{rendererClass=&NiagaraMeshRendererProperties, rendererIndex=0}]
 }}
summary {
 systemName=NS_Mesh_Wind
 emitters [{emitterName=Mesh_Wind, rendererClasses [&NiagaraMeshRendererProperties]}]
 userVariables []
}
