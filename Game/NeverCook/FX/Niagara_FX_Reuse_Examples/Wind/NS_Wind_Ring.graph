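# NS_Wind_Ring is a Niagara wind ring effect from a set of reuse examples, drawn by a single mesh-renderer emitter.
# An instantaneous burst spawns the mesh particle, heavily flattened on Z and given a fixed initial rotation.
# Over particle life, a cubic float curve drives uniform mesh scale from zero to full size, rising quickly at first.
# The system exposes no user parameters.
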NiagaraSystem /Game/NeverCook/FX/Niagara_FX_Reuse_Examples/Wind/NS_Wind_Ring
dependencies {usedDataInterfaces [&NiagaraDataInterfaceCurve], usedDynamicInputs [], usedModules [&SystemState]}
emitters {Wind_Ring {
  emitterUpdateScript [
   &EmitterState(Life Cycle Mode=::NewEnumerator1, Loop Behavior=::NewEnumerator1, Loop Duration=1, UseLoopDelay=false)
   &SpawnBurst_Instantaneous(Use Spawn Probability=false, UseLoopCountLimit=false)
  ]
  particleSpawnScript [
   &InitializeParticle(Color Mode=::NewEnumerator0, Mesh Scale=<1 1 0.02>, Mesh Scale Mode=::NewEnumerator4, UsePositionOffset=false, Write Material Random=true)
   &InitialMeshOrientation(Mesh Orientation Mode=::NewEnumerator2, Rotation=<0 0.5 0>, Use Rotation Vector=true)
  ]
  particleUpdateScript [
   &ParticleState()
   &ScaleMeshSize(Scale Factor=&VectorFromFloat(Value=&FloatFromCurve(FloatCurve=DataInterface() {propertyValues {Curve {keys [{arriveTangent=2.6167, interpMode=RCIM_Cubic, leaveTangent=2.6167, tangentMode=RCTM_User, time=0, value=0}, <at 1 1 RCIM_Cubic>]}}})))
  ]
  rendererClasses [&NiagaraMeshRendererProperties]
  renderers [{rendererClass=&NiagaraMeshRendererProperties, rendererIndex=0}]
 }}
summary {
 systemName=NS_Wind_Ring
 emitters [{emitterName=Wind_Ring, rendererClasses [&NiagaraMeshRendererProperties]}]
 userVariables []
}
